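; Fresh grave hiding prop Blueprint that controls player input and rotation while a player climbs in and out.
; Overrides of adding and removing hidden players call UpdateHidingPawn, which ignores move input until a timer matching the enter or exit montage length fires.
; Hiding also turns off rotate-to-camera until the exit timer restores it.
; An Is Exiting flag blocks interaction during the exit, on top of the parent's interact check.
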(event Custom|UnlockHideInputs
  (Input|SetIgnoreMoveInput (Fortnite|GetFortPlayerControllerfromActor (stale Custom|UpdateHiding%20Pawn _asfort_player_pawn)) false))

(event Custom|UpdateHiding%20Pawn
  (params (Hiding%20Fort%20Pawn :type Fort%20Pawn%20Object%20Reference) (Is%20Hiding :type Boolean))
  (Variables|Default|SetIs%20Hiding Is%20Hiding)
  (Variables|Default|SetIs%20Exiting false)
  (bind _asfort_player_pawn (Utilities|Casting|CastToFortPlayerPawn Hiding%20Fort%20Pawn))
  (if (Variables|Default|GetIs%20Hiding)
    (Input|SetRotatePawntoCamera (Fortnite|GetFortPlayerControllerfromActor _asfort_player_pawn) false)
    (Input|SetIgnoreMoveInput (Fortnite|GetFortPlayerControllerfromActor _asfort_player_pawn) true)
    (Development|PrintString "@@@ UpdateHidingPawn: Locking rotation!" false (Variables|Debug|GetbDebugProp))
    (Utilities|Time|SetTimerbyEvent (delegate Custom|UnlockHideInputs) (Animation|GetPlayLength (Variables|Animation|GetEnterAnimMontage)))
    (else
      (Variables|Default|SetIs%20Exiting true)
      (Input|SetIgnoreMoveInput (Fortnite|GetFortPlayerControllerfromActor _asfort_player_pawn) true)
      (Utilities|Time|SetTimerbyEvent (delegate Custom|UnlockUnhideInputs) (Animation|GetPlayLength (Variables|Animation|GetExitAnimMontage))))))

(event Custom|UnlockUnhideInputs
  (Input|SetRotatePawntoCamera (Fortnite|GetFortPlayerControllerfromActor (stale Custom|UpdateHiding%20Pawn _asfort_player_pawn)) true)
  (Input|SetIgnoreMoveInput (Fortnite|GetFortPlayerControllerfromActor (stale Custom|UpdateHiding%20Pawn _asfort_player_pawn)) false)
  (Variables|Default|SetIs%20Exiting false)
  (Development|PrintString "@@@ UpdateHidingPawn: Unlocking rotation!" false (Variables|Debug|GetbDebugProp)))

(event EventBeginPlay
  (|CallParentFunction) ; This node is disabled and will not be called.\nDrag off pins to build functionality.
)

(event Collision|EventActorBeginOverlap
  (params (OtherActor :type Actor%20Object%20Reference))
  (|CallParentFunction OtherActor) ; This node is disabled and will not be called.\nDrag off pins to build functionality.
)

(fn AddHiddenPlayer
  (params (FortPawn :type Fort%20Pawn%20Object%20Reference))
  (|CallParentFunction FortPawn)
  (Development|PrintString (Utilities|String|Append "@@@ Add Hidden Player [B_HidingProp_FreshGrave]: CALLED, pawn = " (Utilities|GetDisplayName FortPawn) "...") false (Variables|Debug|GetbDebugProp))
  (CallFunction|UpdateHidingPawn (Variables|GetFortPawn) true))

(fn BlueprintCanInteract
  (params (InteractingPawn :type Fort%20Pawn%20Object%20Reference) (InteractionBeingAttempted :type EInteractionBeingAttempted%20Enum :default "FirstInteraction") (InteractionType :type TInteractionType%20Enum :default "IT_NoInteraction"))
  (return (and (not (Variables|Default|GetIs%20Exiting)) (|CallParentFunction InteractingPawn InteractionBeingAttempted InteractionType))))

(fn RemoveHiddenPlayer
  (params (FortPawn :type Fort%20Pawn%20Object%20Reference))
  (|CallParentFunction FortPawn)
  (Development|PrintString (Utilities|String|Append "@@@ Remove Hidden Player [B_HidingProp_FreshGrave]: CALLED, pawn = " (Utilities|GetDisplayName FortPawn) "...") false (Variables|Debug|GetbDebugProp))
  (CallFunction|UpdateHidingPawn (Variables|GetFortPawn)))

(fn ConstructionScript
  (|CallParentFunction))
